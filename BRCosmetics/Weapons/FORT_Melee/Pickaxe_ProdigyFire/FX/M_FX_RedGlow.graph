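# Unlit additive material for Niagara mesh particles, used as a glow effect on the ProdigyFire pickaxe, with color from particle color through an emissive-adjustment function.
# A procedural UV gradient, whose power falloff varies along U, is multiplied by particle alpha and a view-angle fade.
# Named reroutes define three further smoothstep masks: a U wipe driven by dynamic Transition, a V mask sized by dynamic BottomMask, and a front-edge U fade.
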Material /BRCosmetics/Weapons/FORT_Melee/Pickaxe_ProdigyFire/FX/M_FX_RedGlow
graph {
 properties {
  bUsedWithNiagaraMeshParticles=true
  blendMode=BLEND_Additive
  pixelDepthOffsetMode=PDOM_Legacy
  shadingModel=MSM_Unlit
  usedShadingModels=Unlit
 }
 nodes [
  Constant()
  Constant() {r=0.08}
  n1 = ParticleColor()
  MaterialFunctionCall(Emissive=@n1.RGB) {materialFunction=&MF_FX_EmissiveAdjustments}
  n2 = MaterialFunctionCall(Float2=TextureCoordinate()) {materialFunction=&BreakOutFloat2Components}
  n3 = Constant() {r=0.398}
  n4 = SmoothStep(Min=@n3, Max=Add(A=@n3, B=Constant() {r=1}), Value=@n2.R)
  n5 = Constant()
  n6 = Constant() {r=1}
  n7 = ScalarParameter() {defaultValue=3}
  n8 = ScalarParameter() {defaultValue=6, parameterName=Param_3}
  n9 = Constant()
  n10 = Constant() {r=0.08}
  n11 = Multiply(
   A=Power(
    Base=@n2.G
    Exp=MaterialFunctionCall(Input=@n4, Input Low=@n5, Input High=@n6, Target Low=@n7, Target High=@n8) {materialFunction=&RemapValueRange}.Result
   )
   B=SmoothStep(Min=@n9, Max=Add(A=@n9, B=@n10), Value=OneMinus(None=@n2.R))
  )
  n12 = Multiply(
   A=Saturate(None=Multiply(A=@n11, B=@n1.A))
   B=MaterialFunctionCall("FadeSensitivity (optional)"=ScalarParameter() {defaultValue=2, parameterName=Angle Fade}) {materialFunction=&FadeBasedOnViewAngle}.Result
  )
  n13 = NamedRerouteUsage()
  Multiply(A=Multiply(A=Multiply(A=@n12, B=@n13), B=NamedRerouteUsage()), B=NamedRerouteUsage())
  n14 = DynamicParameter() {defaultValue=<rgba 1 0 1 1>, paramNames [Transition,BottomMask,Param3,Param4]}
  n15 = Add(A=@n14.Transition, B=ScalarParameter() {defaultValue=0.2, parameterName=Softness})
  n16 = OneMinus(None=MaterialFunctionCall(Float2=TextureCoordinate()) {materialFunction=&BreakOutFloat2Components}.R)
  NamedRerouteDeclaration(None=Saturate(None=SmoothStep(Min=@n14.Transition, Max=@n15, Value=@n16))) {name=Mask, nodeColor=<rgba 0 0.4609 1 1>}
  n17 = Constant() {r=-0.05}
  n18 = DynamicParameter() {defaultValue=<rgba 1 0 1 1>, paramNames [Transition,BottomMask,Param3,Param4]}
  n19 = MaterialFunctionCall(Float2=TextureCoordinate()) {materialFunction=&BreakOutFloat2Components}
  NamedRerouteDeclaration(None=SmoothStep(Min=@n17, Max=Add(A=@n17, B=@n18.BottomMask), Value=OneMinus(None=@n19.G))) {name=BottomMask, nodeColor=<rgba 0 1 0.9531 1>}
  n20 = ScalarParameter() {parameterName=asdsafda}
  n21 = ScalarParameter() {defaultValue=0.02, parameterName=fasdfa}
  n22 = TextureCoordinate()
  NamedRerouteDeclaration(None=SmoothStep(
    Min=@n20
    Max=Add(A=@n20, B=@n21)
    Value=MaterialFunctionCall(Float2=@n22) {materialFunction=&BreakOutFloat2Components}.R
   )) {name=FrontMask, nodeColor=<rgba 1 0 0.6562 1>}
 ]
}
